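# Material_MaterialsTest is a minimal Default Lit material in the EntityFrameworkTests MaterialsTest folder, apparently a test asset.
# Its graph holds a single Constant node at its default value, with no connection to any material output shown.
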Material /EntityFrameworkTests/MaterialsTest/Material_MaterialsTest
graph {properties {usedShadingModels=Default Lit}, nodes [Constant()]}
